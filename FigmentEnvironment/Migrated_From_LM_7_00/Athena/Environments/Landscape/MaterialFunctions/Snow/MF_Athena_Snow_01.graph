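# Landscape material function for Athena snow, outputting material attributes plus base color, normal, height, height mask and view-facing outputs.
# World XY projects a packed snow texture that drives color, flake roughness and specular, plus a flattened macro normal.
# SharedWinter snow color is tinted by a view-angle LUT, with cheaper low-quality and mobile paths; the LUT's sun-direction lerp has identical inputs.
# Height blends toward a constant by terrain distance fade.
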MaterialFunction /FigmentEnvironment/Migrated_From_LM_7_00/Athena/Environments/Landscape/MaterialFunctions/Snow/MF_Athena_Snow_01
graph {nodes [
  CollectionParameter() {
   collection=/FigmentEnvironment/Migrated_From_LM_7_00/Packages/Fortress_SharedMaterials/GlobalMaterialParameters/FortniteMaterialParameters
   parameterName=SunlightDirectionalVector
  }
  FeatureLevelSwitch()
  n1 = CameraVectorWS()
  n2 = DotProduct(A=VertexNormalWS(), B=@n1)
  n3 = AppendVector(A=Constant() {r=0.5}, B=@n2)
  n4 = Constant() {r=0.7}
  n5 = TextureSample(UVs=QualitySwitch(Default=@n3, Low=@n4, High=@n3, Medium=@n4)) {
   samplerSource=SSM_Clamp_WorldGroupSettings
   samplerType=SAMPLERTYPE_LinearColor
   texture=&/LUT_Snow
  }
  n6 = SkyAtmosphereLightDirection()
  n7 = Clamp(None=DotProduct(A=@n1, B=ComponentMask(None=@n6) {b=true, g=true, r=true}))
  n8 = Constant() {r=0.5}
  n9 = Divide(
   A=ComponentMask(None=WorldPosition().XYZ) {g=true, r=true}
   B=ScalarParameter() {defaultValue=1200, group=Snow, parameterName=Snow_NormalSize}
  )
  n10 = TextureSample(UVs=@n9) {
   samplerSource=SSM_Wrap_WorldGroupSettings
   samplerType=SAMPLERTYPE_LinearColor
   texture=/FigmentEnvironment/Migrated_From_LM_7_00/Athena/Environments/Landscape/Textures/T_Athena_Snow_01_M
  }
  n11 = Multiply(
   A=MaterialFunctionCall() {materialFunction=/FigmentEnvironment/Migrated_From_LM_7_00/Athena/WinterArt/Materials/MF_SharedWinter}.SnowColor
   B=@n10.B
  )
  n12 = Multiply(A=@n10.B) {constB=0.7}
  n13 = QualitySwitch(Default=@n11, Low=@n12, High=@n11, Medium=@n12)
  n14 = FunctionOutput(None=FeatureLevelSwitch(Default=@n13, ES2=@n12, ES3_1=@n12, SM4_REMOVED=@n13, SM5=@n13, SM6=@n13)) {outputName=BaseColor}
  n15 = Multiply(A=LinearInterpolate(A=@n5.RGB, B=@n5.RGB, Alpha=QualitySwitch(Default=@n7, Low=@n8, High=@n7, Medium=@n8)), B=@n14)
  n16 = LinearInterpolate(
   A=ScalarParameter() {defaultValue=0.2, group=Snow, parameterName=SnowRoughnessFlakes}
   B=ScalarParameter() {defaultValue=0.8, group=Snow, parameterName=SnowRoughness}
   Alpha=@n10.R
  )
  n17 = TextureSample(UVs=Multiply(A=@n9) {constB=0.15}) {
   samplerSource=SSM_Wrap_WorldGroupSettings
   samplerType=SAMPLERTYPE_Normal
   texture=/Game/ContentCreationTools/TextureCreation/Textures/T_BPCreated_MacroNormal_01
  }
  n18 = LinearInterpolate(A=@n17.RGB, B=Constant3Vector() {constant=<rgba 0 0 1 0>}, Alpha=Constant() {r=0.7}) {constAlpha=0.8}
  FunctionOutput(None=MakeMaterialAttributes!(
    BaseColor=QualitySwitch(Default=@n15, Low=@n14, High=@n15, Medium=@n14)
    Metallic
    Specular=LinearInterpolate(Alpha=@n10.R) {constA=1, constB=0.35}
    Roughness=Clamp(None=@n16)
    Anisotropy
    EmissiveColor
    Opacity
    OpacityMask
    Normal=FunctionOutput(None=@n18) {outputName=Normal}
    Tangent
    WorldPositionOffset
    SubsurfaceColor
    ClearCoat
    ClearCoatRoughness
    AmbientOcclusion
    Refraction
    CustomizedUVs_0
    CustomizedUVs_1
    CustomizedUVs_2
    CustomizedUVs_3
    CustomizedUVs_4
    CustomizedUVs_5
    CustomizedUVs_6
    CustomizedUVs_7
    PixelDepthOffset
    ShadingModel
    Displacement
   )) {outputName=Material}
  n19 = TextureSample(UVs=Multiply(A=@n9, B=ScalarParameter() {defaultValue=0.15, parameterName=TestingScaleHeight})) {
   samplerSource=SSM_Wrap_WorldGroupSettings
   samplerType=SAMPLERTYPE_LinearColor
   texture=&T_Athena_Snow_01_M
  }
  n20 = LinearInterpolate(
   A=@n10.A
   B=ScalarParameter() {defaultValue=0.5, group=Snow, parameterName=TestingHeight}
   Alpha=MaterialFunctionCall!(Distance=ScalarParameter() {defaultValue=90000, parameterName=Snow_Distance_Height}) {materialFunction=/FigmentEnvironment/Migrated_From_LM_7_00/Athena/Environments/Landscape/MaterialFunctions/MF_TerrainDistanceFade}.Result
  )
  FunctionOutput(None=Multiply(A=@n19.G, B=@n20)) {outputName=Snow_Mask_Height}
  FunctionOutput(None=@n20) {outputName=Height}
  FunctionOutput(None=@n2) {outputName=NormalDotCamera}
  PixelNormalWS()
 ]}
